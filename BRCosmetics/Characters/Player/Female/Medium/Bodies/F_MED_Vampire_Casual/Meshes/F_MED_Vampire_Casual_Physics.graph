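# Physics asset for the F_MED_Vampire_Casual female player cosmetic body, setting up physics for its shirt and belt bones.
# Boxes and a capsule cover four dyn_ shirt bones and one belt bone, while capsules approximate the pelvis, spine and left thigh only.
# Shirt bones attach directly to spine_04 and the belt to the pelvis with angular limits, tight in twist and wider in swing2.
# Spine and thigh constraints leave all angular motion free.
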PhysicsAsset /BRCosmetics/Characters/Player/Female/Medium/Bodies/F_MED_Vampire_Casual/Meshes/F_MED_Vampire_Casual_Physics
bodies {
 dyn_b_shirt [{shapeType=Capsule, length=9.3928, radius=1.9593, center=<1.8253 0 0>, rotation=<rot 0 0 90>}]
 dyn_belt_01 [{shapeType=Box, extentX=5.8263, extentY=1.01, extentZ=3.3821, center=<3.0734 0.6065 0.0173>, rotation=<rot 0.3853 13.5938 13.6028>}]
 dyn_f_shirt [{shapeType=Box, extentX=6.242, extentY=12, extentZ=0.5, center=<3.0184 0 0>}]
 dyn_l_shirt [{shapeType=Box, extentX=5.9179, extentY=12, extentZ=1.3443, center=<2.8099 -0.0001 0>}]
 dyn_r_shirt [{shapeType=Box, extentX=5.4819, extentY=12, extentZ=0.5, center=<2.9373 -0.0001 0>}]
 pelvis [{shapeType=Capsule, length=9.7162, radius=9.3221}]
 spine_01 [{shapeType=Capsule, length=6.8512, radius=7.7468}]
 spine_02 [{shapeType=Capsule, length=7.1268, radius=7.4713, center=<0 1.7777 0>}]
 spine_03 [{shapeType=Capsule, length=6.04, radius=7.0459, center=<0 2.1796 0>}]
 spine_04 [{shapeType=Capsule, length=7.2454, radius=6.9963, center=<0 3.221 0>}]
 thigh_l [{shapeType=Capsule, length=28.5746, radius=7.8984, center=<-20.6428 0.0002 0>, rotation=<rot 90 0 0>}]
}
constraints [
 {
  bone1Name=spine_01
  bone2Name=pelvis
  swing1LimitDegrees=45
  swing1Motion=Free
  swing2LimitDegrees=45
  swing2Motion=Free
  twistLimitDegrees=45
  twistMotion=Free
 }
 {
  bone1Name=spine_02
  bone2Name=spine_01
  swing1LimitDegrees=45
  swing1Motion=Free
  swing2LimitDegrees=45
  swing2Motion=Free
  twistLimitDegrees=45
  twistMotion=Free
 }
 {
  bone1Name=spine_03
  bone2Name=spine_02
  swing1LimitDegrees=45
  swing1Motion=Free
  swing2LimitDegrees=45
  swing2Motion=Free
  twistLimitDegrees=45
  twistMotion=Free
 }
 {
  bone1Name=spine_04
  bone2Name=spine_03
  swing1LimitDegrees=45
  swing1Motion=Free
  swing2LimitDegrees=45
  swing2Motion=Free
  twistLimitDegrees=45
  twistMotion=Free
 }
 {bone1Name=dyn_r_shirt, bone2Name=spine_04, swing1LimitDegrees=10, swing2LimitDegrees=50, twistLimitDegrees=5}
 {bone1Name=dyn_f_shirt, bone2Name=spine_04, swing1LimitDegrees=10, swing2LimitDegrees=50, twistLimitDegrees=5}
 {bone1Name=dyn_b_shirt, bone2Name=spine_04, swing1LimitDegrees=10, swing2LimitDegrees=50, twistLimitDegrees=5}
 {bone1Name=dyn_l_shirt, bone2Name=spine_04, swing1LimitDegrees=10, swing2LimitDegrees=50, twistLimitDegrees=5}
 {bone1Name=dyn_belt_01, bone2Name=pelvis, swing1LimitDegrees=10, swing2LimitDegrees=45, twistLimitDegrees=5}
 {
  bone1Name=thigh_l
  bone2Name=pelvis
  swing1LimitDegrees=45
  swing1Motion=Free
  swing2LimitDegrees=45
  swing2Motion=Free
  twistLimitDegrees=45
  twistMotion=Free
 }
]
