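# Post-copy secondary-motion AnimBP for the female Vampire Casual body, copying the parent mesh pose and layering cosmetic physics and fixes.
# A weighted spine/pelvis constraint drives the belt, then RigidBody simulates in pelvis space.
# Chained AnimDynamics nodes add motion to shoelaces, shirt and bracelets.
# Final bone offsets correct torso, collar, left thigh (crouching) and right knee (jumping or falling), using state variables not declared here.
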AnimBlueprint /BRCosmetics/Characters/Player/Female/Medium/Bodies/F_MED_Vampire_Casual/Meshes/F_MED_Vampire_Casual_AnimBP
target_skeleton=&/F_MED_Vampire_Casual_Skeleton
graphs [
 AnimGraph {nodes [
   n1 = "Constraint - dyn_belt_MSTR (PPPPP)"!(
    ComponentPose=Local To Component(LocalPose=Copy Pose From Mesh() {bCopyCurves=true, bUseAttachedParent=true})
    Alpha=1
    ConstraintWeights_0=0.1862
    ConstraintWeights_1:float=0.25
    ConstraintWeights_2:float=0.0181
    ConstraintWeights_3:float=0.0348
    ConstraintWeights_4:float=0.5108
   ) -> (Pose) {
    alphaBoolBlend {blendOption=Linear}
    boneToModify {boneName=dyn_belt_MSTR}
    constraintSetup [
     {targetBone {boneName=spine_01}}
     {targetBone {boneName=spine_02}}
     {targetBone {boneName=spine_03}}
     {targetBone {boneName=spine_04}}
     {targetBone {boneName=pelvis}}
    ]
    constraintWeights [1,1,1,1,1]
   }
   n2 = RigidBody(ComponentPose=@n1) {
    bClampLinearTranslationLimitToRefPose=true
    bOverrideWorldGravity=true
    overridePhysicsAsset=&/F_MED_Vampire_Casual_Physics
    simulationSpace=BaseBoneSpace
    alphaBoolBlend {blendOption=Linear}
    baseBoneRef {boneName=pelvis}
    componentAppliedLinearAccClamp=<1000 1000 1000>
    componentLinearAccScale=<0.5 0.5 0.42>
    overrideWorldGravity {z=-980}
   }
   n3 = Anim Dynamics - Bone: dyn_l_shoeLace_a_01!(ComponentPose=@n2, Alpha=1, GravityOverride:Vector="0,0,0", ExternalForce="0,0,0") -> (Pose) {
    angularDampingOverride=1
    angularSpringConstant=1500
    bAngularSpring=true
    bGravityOverrideInSimSpace=true
    bOverrideAngularDamping=true
    gravityScale=0.35
    simulationSpace=RootRelative
    alphaBoolBlend {blendOption=Linear}
    alphaScaleBias {scale=2}
    boundBone {boneName=dyn_l_shoeLace_a_01}
    physicsBodyDefinitions [{
      sphereCollisionRadius=0
      boundBone {boneName=dyn_l_shoeLace_a_01}
      boxExtents=<10 20 20>
      constraintSetup {angularLimitsMax=<10 10 10>, angularLimitsMin=<-10 -20 -10>, angularTarget=<0 0 0>}
      localJointOffset=<10 0 0>
     }]
   }
   n4 = Anim Dynamics - Bone: dyn_l_shoeLace_a_02!(ComponentPose=@n3, Alpha=1, GravityOverride:Vector="0,0,0", ExternalForce="0,0,0") -> (Pose) {
    angularDampingOverride=1
    angularSpringConstant=1500
    bAngularSpring=true
    bGravityOverrideInSimSpace=true
    bOverrideAngularDamping=true
    gravityScale=0.35
    simulationSpace=RootRelative
    alphaBoolBlend {blendOption=Linear}
    alphaScaleBias {scale=2}
    boundBone {boneName=dyn_l_shoeLace_a_02}
    physicsBodyDefinitions [{
      sphereCollisionRadius=0
      boundBone {boneName=dyn_l_shoeLace_a_02}
      boxExtents=<10 20 20>
      constraintSetup {angularLimitsMax=<10 10 10>, angularLimitsMin=<-10 -20 -10>, angularTarget=<0 0 0>}
      localJointOffset=<10 0 0>
     }]
   }
   n5 = Anim Dynamics - Bone: dyn_l_shoeLace_a_03!(ComponentPose=@n4, Alpha=1, GravityOverride:Vector="0,0,0", ExternalForce="0,0,0") -> (Pose) {
    angularDampingOverride=1
    angularSpringConstant=1500
    bAngularSpring=true
    bGravityOverrideInSimSpace=true
    bOverrideAngularDamping=true
    gravityScale=0.35
    simulationSpace=RootRelative
    alphaBoolBlend {blendOption=Linear}
    alphaScaleBias {scale=2}
    boundBone {boneName=dyn_l_shoeLace_a_03}
    physicsBodyDefinitions [{
      sphereCollisionRadius=0
      boundBone {boneName=dyn_l_shoeLace_a_03}
      boxExtents=<10 20 20>
      constraintSetup {angularLimitsMax=<10 0 10>, angularLimitsMin=<-10 -20 -10>, angularTarget=<0 0 0>}
      localJointOffset=<10 0 0>
     }]
   }
   n6 = Anim Dynamics - Bone: dyn_l_shoeLace_b_01!(ComponentPose=@n5, Alpha=1, GravityOverride:Vector="0,0,0", ExternalForce="0,0,0") -> (Pose) {
    angularDampingOverride=1
    angularSpringConstant=1500
    bAngularSpring=true
    bGravityOverrideInSimSpace=true
    bOverrideAngularDamping=true
    gravityScale=0.35
    simulationSpace=RootRelative
    alphaBoolBlend {blendOption=Linear}
    alphaScaleBias {scale=2}
    boundBone {boneName=dyn_l_shoeLace_b_01}
    physicsBodyDefinitions [{
      sphereCollisionRadius=0
      boundBone {boneName=dyn_l_shoeLace_b_01}
      boxExtents=<10 20 20>
      constraintSetup {angularLimitsMax=<10 10 10>, angularLimitsMin=<-10 -20 -10>, angularTarget=<0 0 0>}
      localJointOffset=<10 0 0>
     }]
   }
   n7 = Anim Dynamics - Bone: dyn_l_shoeLace_b_02!(ComponentPose=@n6, Alpha=1, GravityOverride:Vector="0,0,0", ExternalForce="0,0,0") -> (Pose) {
    angularDampingOverride=1
    angularSpringConstant=1500
    bAngularSpring=true
    bGravityOverrideInSimSpace=true
    bOverrideAngularDamping=true
    gravityScale=0.35
    simulationSpace=RootRelative
    alphaBoolBlend {blendOption=Linear}
    alphaScaleBias {scale=2}
    boundBone {boneName=dyn_l_shoeLace_b_02}
    physicsBodyDefinitions [{
      sphereCollisionRadius=0
      boundBone {boneName=dyn_l_shoeLace_b_02}
      boxExtents=<10 20 20>
      constraintSetup {angularLimitsMax=<10 10 10>, angularLimitsMin=<-10 -20 -10>, angularTarget=<0 0 0>}
      localJointOffset=<10 0 0>
     }]
   }
   n8 = Anim Dynamics - Bone: dyn_l_shoeLace_b_03!(ComponentPose=@n7, Alpha=1, GravityOverride:Vector="0,0,0", ExternalForce="0,0,0") -> (Pose) {
    angularDampingOverride=1
    angularSpringConstant=1500
    bAngularSpring=true
    bGravityOverrideInSimSpace=true
    bOverrideAngularDamping=true
    gravityScale=0.35
    simulationSpace=RootRelative
    alphaBoolBlend {blendOption=Linear}
    alphaScaleBias {scale=2}
    boundBone {boneName=dyn_l_shoeLace_b_03}
    physicsBodyDefinitions [{
      sphereCollisionRadius=0
      boundBone {boneName=dyn_l_shoeLace_b_03}
      boxExtents=<10 20 20>
      constraintSetup {angularLimitsMax=<10 0 10>, angularLimitsMin=<-10 -20 -10>, angularTarget=<0 0 0>}
      localJointOffset=<10 0 0>
     }]
   }
   n9 = Anim Dynamics - Bone: dyn_r_shoeLace_a_01!(ComponentPose=@n8, Alpha=1, GravityOverride:Vector="0,0,0", ExternalForce="0,0,0") -> (Pose) {
    angularDampingOverride=1
    angularSpringConstant=1500
    bAngularSpring=true
    bGravityOverrideInSimSpace=true
    bOverrideAngularDamping=true
    gravityScale=0.35
    simulationSpace=RootRelative
    alphaBoolBlend {blendOption=Linear}
    alphaScaleBias {scale=2}
    boundBone {boneName=dyn_r_shoeLace_a_01}
    physicsBodyDefinitions [{
      sphereCollisionRadius=0
      boundBone {boneName=dyn_r_shoeLace_a_01}
      boxExtents=<10 20 20>
      constraintSetup {angularLimitsMax=<10 10 10>, angularLimitsMin=<-10 -20 -10>, angularTarget=<0 0 0>}
      localJointOffset=<-10 0 0>
     }]
   }
   n10 = Anim Dynamics - Bone: dyn_r_shoeLace_a_03!(ComponentPose=@n9, Alpha=1, GravityOverride:Vector="0,0,0", ExternalForce="0,0,0") -> (Pose) {
    angularDampingOverride=1
    angularSpringConstant=1500
    bAngularSpring=true
    bGravityOverrideInSimSpace=true
    bOverrideAngularDamping=true
    gravityScale=0.35
    simulationSpace=RootRelative
    alphaBoolBlend {blendOption=Linear}
    alphaScaleBias {scale=2}
    boundBone {boneName=dyn_r_shoeLace_a_03}
    physicsBodyDefinitions [{
      sphereCollisionRadius=0
      boundBone {boneName=dyn_r_shoeLace_a_03}
      boxExtents=<10 20 20>
      constraintSetup {angularLimitsMax=<10 0 10>, angularLimitsMin=<-10 -20 -10>, angularTarget=<0 0 0>}
      localJointOffset=<-10 0 0>
     }]
   }
   n11 = Anim Dynamics - Bone: dyn_r_shoeLace_b_02!(ComponentPose=@n10, Alpha=1, GravityOverride:Vector="0,0,0", ExternalForce="0,0,0") -> (Pose) {
    angularDampingOverride=1
    angularSpringConstant=1500
    bAngularSpring=true
    bGravityOverrideInSimSpace=true
    bOverrideAngularDamping=true
    gravityScale=0.35
    simulationSpace=RootRelative
    alphaBoolBlend {blendOption=Linear}
    alphaScaleBias {scale=2}
    boundBone {boneName=dyn_r_shoeLace_b_02}
    physicsBodyDefinitions [{
      sphereCollisionRadius=0
      boundBone {boneName=dyn_r_shoeLace_b_02}
      boxExtents=<10 20 20>
      constraintSetup {angularLimitsMax=<10 10 10>, angularLimitsMin=<-10 -20 -10>, angularTarget=<0 0 0>}
      localJointOffset=<-10 0 0>
     }]
   }
   n12 = Anim Dynamics - Bone: dyn_r_shoeLace_a_02!(ComponentPose=@n11, Alpha=1, GravityOverride:Vector="0,0,0", ExternalForce="0,0,0") -> (Pose) {
    angularDampingOverride=1
    angularSpringConstant=1500
    bAngularSpring=true
    bGravityOverrideInSimSpace=true
    bOverrideAngularDamping=true
    gravityScale=0.35
    simulationSpace=RootRelative
    alphaBoolBlend {blendOption=Linear}
    alphaScaleBias {scale=2}
    boundBone {boneName=dyn_r_shoeLace_a_02}
    physicsBodyDefinitions [{
      sphereCollisionRadius=0
      boundBone {boneName=dyn_r_shoeLace_a_02}
      boxExtents=<10 20 20>
      constraintSetup {angularLimitsMax=<10 10 10>, angularLimitsMin=<-10 -20 -10>, angularTarget=<0 0 0>}
      localJointOffset=<-10 0 0>
     }]
   }
   n13 = Anim Dynamics - Bone: dyn_r_shoeLace_b_01!(ComponentPose=@n12, Alpha=1, GravityOverride:Vector="0,0,0", ExternalForce="0,0,0") -> (Pose) {
    angularDampingOverride=1
    angularSpringConstant=1500
    bAngularSpring=true
    bGravityOverrideInSimSpace=true
    bOverrideAngularDamping=true
    gravityScale=0.35
    simulationSpace=RootRelative
    alphaBoolBlend {blendOption=Linear}
    alphaScaleBias {scale=2}
    boundBone {boneName=dyn_r_shoeLace_b_01}
    physicsBodyDefinitions [{
      sphereCollisionRadius=0
      boundBone {boneName=dyn_r_shoeLace_b_01}
      boxExtents=<10 20 20>
      constraintSetup {angularLimitsMax=<10 10 10>, angularLimitsMin=<-10 -20 -10>, angularTarget=<0 0 0>}
      localJointOffset=<-10 0 0>
     }]
   }
   n14 = Anim Dynamics - Bone: dyn_r_shoeLace_b_03!(ComponentPose=@n13, Alpha=1, GravityOverride:Vector="0,0,0", ExternalForce="0,0,0") -> (Pose) {
    angularDampingOverride=1
    angularSpringConstant=1500
    bAngularSpring=true
    bGravityOverrideInSimSpace=true
    bOverrideAngularDamping=true
    gravityScale=0.35
    simulationSpace=RootRelative
    alphaBoolBlend {blendOption=Linear}
    alphaScaleBias {scale=2}
    boundBone {boneName=dyn_r_shoeLace_b_03}
    physicsBodyDefinitions [{
      sphereCollisionRadius=0
      boundBone {boneName=dyn_r_shoeLace_b_03}
      boxExtents=<10 20 20>
      constraintSetup {angularLimitsMax=<10 0 10>, angularLimitsMin=<-10 -20 -10>, angularTarget=<0 0 0>}
      localJointOffset=<-10 0 0>
     }]
   }
   n15 = Anim Dynamics - Bone: dyn_r_shirt(ComponentPose=@n14, Alpha=1, ExternalForce="0,0,0") {
    angularDampingOverride=0.9999
    angularSpringConstant=25
    bAngularSpring=true
    bGravityOverrideInSimSpace=true
    bOverrideAngularDamping=true
    simulationSpace=BoneRelative
    alphaBoolBlend {blendOption=Linear}
    boundBone {boneName=dyn_r_shirt}
    componentAppliedLinearAccClamp=<3500 3500 3500>
    componentLinearAccScale=<1.5 1.5 1.5>
    componentLinearVelScale=<1.5 1.5 1.5>
    physicsBodyDefinitions [{
      sphereCollisionRadius=0
      boundBone {boneName=dyn_r_shirt}
      boxExtents=<10 20 20>
      constraintSetup {angularLimitsMax=<5 0 5>, angularLimitsMin=<-5 -30 -5>, angularTarget=<0 0 0>}
      localJointOffset=<2 0 0>
     }]
    relativeSpaceBone {boneName=root}
   }
   n16 = Anim Dynamics - Bone: dyn_f_shirt(ComponentPose=@n15, Alpha=1, ExternalForce="0,0,0") {
    angularDampingOverride=0.9999
    angularSpringConstant=25
    bAngularSpring=true
    bGravityOverrideInSimSpace=true
    bOverrideAngularDamping=true
    simulationSpace=BoneRelative
    alphaBoolBlend {blendOption=Linear}
    boundBone {boneName=dyn_f_shirt}
    componentAppliedLinearAccClamp=<3500 3500 3500>
    componentLinearAccScale=<1.5 1.5 1.5>
    componentLinearVelScale=<1.5 1.5 1.5>
    physicsBodyDefinitions [{
      sphereCollisionRadius=0
      boundBone {boneName=dyn_f_shirt}
      boxExtents=<10 20 20>
      constraintSetup {angularLimitsMax=<5 0 5>, angularLimitsMin=<-5 -30 -5>, angularTarget=<0 0 0>}
      localJointOffset=<2 0 0>
     }]
    relativeSpaceBone {boneName=root}
   }
   n17 = Anim Dynamics - Bone: dyn_l_shirt(ComponentPose=@n16, Alpha=1, ExternalForce="0,0,0") {
    angularDampingOverride=0.9999
    angularSpringConstant=25
    bAngularSpring=true
    bGravityOverrideInSimSpace=true
    bOverrideAngularDamping=true
    simulationSpace=BoneRelative
    alphaBoolBlend {blendOption=Linear}
    boundBone {boneName=dyn_l_shirt}
    componentAppliedLinearAccClamp=<3500 3500 3500>
    componentLinearAccScale=<1.5 1.5 1.5>
    componentLinearVelScale=<1.5 1.5 1.5>
    physicsBodyDefinitions [{
      sphereCollisionRadius=0
      boundBone {boneName=dyn_l_shirt}
      boxExtents=<10 20 20>
      constraintSetup {angularLimitsMax=<5 0 5>, angularLimitsMin=<-5 -30 -5>, angularTarget=<0 0 0>}
      localJointOffset=<2 0 0>
     }]
    relativeSpaceBone {boneName=root}
   }
   n18 = Anim Dynamics - Bone: dyn_a_braclet(ComponentPose=@n17, Alpha=0.2, ExternalForce="0,0,0") {
    angularDampingOverride=0.5
    bGravityOverrideInSimSpace=true
    bOverrideAngularDamping=true
    bOverrideLinearDamping=true
    gravityScale=0.25
    linearDampingOverride=0.5
    alphaBoolBlend {blendOption=Linear}
    boundBone {boneName=dyn_a_braclet}
    physicsBodyDefinitions [{
      sphereCollisionRadius=0
      boundBone {boneName=dyn_a_braclet}
      boxExtents=<2 2 2>
      constraintSetup {
       angularLimitsMax=<15 15 15>
       angularLimitsMin=<-15 -15 -15>
       angularTarget=<0 0 0>
       linearAxesMax=<2 0.15 0.15>
       linearAxesMin=<-2 -0.15 -0.15>
      }
      localJointOffset=<0 0 -0.1>
     }]
    relativeSpaceBone {boneName=lowerarm_r}
   }
   n19 = Anim Dynamics - Bone: dyn_b_braclet(ComponentPose=@n18, Alpha=0.2, ExternalForce="0,0,0") {
    angularDampingOverride=0.5
    bGravityOverrideInSimSpace=true
    bOverrideAngularDamping=true
    bOverrideLinearDamping=true
    gravityScale=0.5
    linearDampingOverride=0.5
    alphaBoolBlend {blendOption=Linear}
    boundBone {boneName=dyn_b_braclet}
    physicsBodyDefinitions [{
      sphereCollisionRadius=0
      boundBone {boneName=dyn_b_braclet}
      boxExtents=<2 2 2>
      constraintSetup {
       angularLimitsMax=<15 15 15>
       angularLimitsMin=<-15 -15 -15>
       angularTarget=<0 0 0>
       linearAxesMax=<2 0.15 0.15>
       linearAxesMin=<-2 -0.15 -0.15>
      }
      localJointOffset=<0 0 -0.1>
     }]
    relativeSpaceBone {boneName=lowerarm_r}
   }
   n20 = "Transform (Modify) Bone - Bone: dyn_torsoFix"(ComponentPose=@n19, Alpha=1, Translation="0,0,0", Rotation="0,0,0", Scale="0.9,0.9,0.9") {
    rotationMode=BMM_Additive
    scaleMode=BMM_Additive
    translationMode=BMM_Additive
    alphaBoolBlend {blendOption=Linear}
    boneToModify {boneName=dyn_torsoFix}
    scale=<0.9 0.9 0.9>
   }
   n21 = "Transform (Modify) Bone - Bone: thigh_l_half"(ComponentPose=@n20, Alpha=Get Crouch_Alpha() -> (Crouch_Alpha:double), Translation="0,0,-3.5", Rotation="0,0,0", Scale="1,1,1") {
    rotationMode=BMM_Additive
    scaleMode=BMM_Additive
    translationMode=BMM_Additive
    alphaBoolBlend {blendOption=Linear}
    boneToModify {boneName=thigh_l_half}
    translation {z=-3.5}
   }
   n22 = Get bIsJumping()
   n23 = Get bIsFalling()
   n24 = "Transform (Modify) Bone - Bone: dyn_r_knee"(ComponentPose=@n21, Alpha="To Float (Boolean)"(InBool=OR Boolean(A=@n22, B=@n23)), Translation="0,0,0", Rotation="0,0,0", Scale="2,1,2") {
    rotationMode=BMM_Additive
    scaleMode=BMM_Additive
    translationMode=BMM_Additive
    alphaBoolBlend {blendOption=Linear}
    boneToModify {boneName=dyn_r_knee}
    scale {x=2, z=2}
   }
   n25 = "Transform (Modify) Bone - Bone: dyn_collarTop"(ComponentPose=@n24, Alpha=1, Translation="0,0,0", Rotation="0,0,0", Scale="1.15,1.15,1.15") {
    rotationMode=BMM_Additive
    scaleMode=BMM_Additive
    translationMode=BMM_Additive
    alphaBoolBlend {blendOption=Linear}
    boneToModify {boneName=dyn_collarTop}
    scale=<1.15 1.15 1.15>
   }
   n26 = "Transform (Modify) Bone - Bone: dyn_collarBtm"(ComponentPose=@n25, Alpha=1, Translation="0,0,0", Rotation="0,0,0", Scale="1.15,1.15,1.15") {
    rotationMode=BMM_Additive
    scaleMode=BMM_Additive
    translationMode=BMM_Additive
    alphaBoolBlend {blendOption=Linear}
    boneToModify {boneName=dyn_collarBtm}
    scale=<1.15 1.15 1.15>
   }
   Output Pose(Result=Component To Local(ComponentPose=@n26)) {layerGroup=None}
  ]}
 EventGraph {nodes [
   Event BlueprintUpdateAnimation()
   > Set Crouch_Alpha(Crouch_Alpha:double=SelectFloat(A=1, B=0, bPickA=Get bIsCrouching())) -> (Output_Get:double)
  ]}
]
variables [Crouch_Alpha {type=number}]
