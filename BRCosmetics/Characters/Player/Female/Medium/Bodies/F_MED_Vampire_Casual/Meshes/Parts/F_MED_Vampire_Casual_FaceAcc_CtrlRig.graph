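# Control Rig for the casual female vampire's face accessory part, nudging right-side hair chains as the head turns relative to neck_01.
# One axis (Euler Z) of that rotation, remapped over negative angles down to -50, drives additive global offsets on the f, g and h hair master bones.
# Offsets are mostly translations; the second h adjustment and the f and g rotations remap to zero, so have no effect.
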ControlRigBlueprint /BRCosmetics/Characters/Player/Female/Medium/Bodies/F_MED_Vampire_Casual/Meshes/Parts/F_MED_Vampire_Casual_FaceAcc_CtrlRig
graphs [
 RigVMFunctionLibrary {links [], nodes []}
 RigVMModel {nodes [
   n1 = &RigUnit_BeginExecution()
   n2 = &RigVMFunction_MathQuaternionToEuler(RotationOrder=ZYX, Value=&RigUnit_GetRelativeBoneTransform(Bone=head, Space=neck_01).Transform.Rotation)
   /Script/ControlRig.RigUnit_ModifyBoneTransforms(
    BoneToModify=((Bone="dyn_hair_g_MSTR",Transform=(Rotation=(X=0,Y=0,Z=0,W=1),Translation=(X=0,Y=0,Z=0),Scale3D=(X=1,Y=1,Z=1))))
    Mode=AdditiveGlobal
    Weight=1
    WeightMaximum=1
    BoneToModify.0.Transform.Translation.Z=&RigVMFunction_MathFloatRemap(SourceMaximum=-50, SourceMinimum=-19, TargetMinimum=0, bClamp=true, Value=@n2.Result.Z).Result
    BoneToModify.0.Transform.Translation.Y=&RigVMFunction_MathFloatRemap(SourceMaximum=-50, SourceMinimum=0, TargetMaximum=-3, TargetMinimum=0, bClamp=true, Value=@n2.Result.Z).Result
    BoneToModify.0.Transform.Translation.X=&RigVMFunction_MathFloatRemap(SourceMaximum=-50, SourceMinimum=-19, TargetMaximum=-1, TargetMinimum=0, bClamp=true, Value=@n2.Result.Z).Result
    BoneToModify.0.Transform.Rotation=&RigVMFunction_MathQuaternionFromEuler(
     Euler=(X=0,Y=0,Z=0)
     Euler.Y=&RigVMFunction_MathFloatRemap(SourceMaximum=-50, SourceMinimum=-19, TargetMaximum=0, TargetMinimum=0, bClamp=true, Value=@n2.Result.Z).Result
    ).Result
    ExecutePin=@n1.ExecutePin
   ) {is=struct RigVMUnitNode}
   n3 = &RigVMFunction_MathFloatRemap(SourceMaximum=-50, SourceMinimum=0, TargetMinimum=0, bClamp=true, Value=@n2.Result.Z)
   > &RigUnit_ModifyBoneTransforms(
    BoneToModify=((Bone="dyn_hair_h_MSTR",Transform=(Rotation=(X=0,Y=0,Z=0,W=1),Translation=(X=0,Y=2,Z=0),Scale3D=(X=1,Y=1,Z=1))))
    Mode=AdditiveGlobal
    WeightMaximum=1
    Weight=@n3.Result
    BoneToModify.0.Transform.Rotation=&RigVMFunction_MathQuaternionFromEuler(Euler=(X=0,Y=0,Z=0), Euler.Y=@n3.Result).Result
   )
   n4 = &RigVMFunction_MathFloatRemap(SourceMaximum=-50, SourceMinimum=-19, TargetMaximum=0, TargetMinimum=0, bClamp=False, Value=@n2.Result.Z)
   > &RigUnit_ModifyBoneTransforms(
    BoneToModify=((Bone="dyn_hair_h_MSTR",Transform=(Rotation=(X=0,Y=0,Z=0,W=1),Translation=(X=0,Y=0,Z=0),Scale3D=(X=1,Y=1,Z=1))))
    Mode=AdditiveGlobal
    Weight=1
    WeightMaximum=1
    BoneToModify.0.Transform.Translation.Z=&RigVMFunction_MathFloatRemap(SourceMaximum=-50, SourceMinimum=-30, TargetMaximum=0, TargetMinimum=0, bClamp=true, Value=@n2.Result.Z).Result
    BoneToModify.0.Transform.Translation.Y=&RigVMFunction_MathFloatRemap(SourceMaximum=-50, SourceMinimum=0, TargetMaximum=0, TargetMinimum=0, bClamp=true, Value=@n2.Result.Z).Result
    BoneToModify.0.Transform.Rotation=&RigVMFunction_MathQuaternionFromEuler(Euler=(X=0,Y=0,Z=0), Euler.Y=@n4.Result, Euler.Z=@n4.Result).Result
   )
   > &RigUnit_ModifyBoneTransforms(
    BoneToModify=((Bone="dyn_hair_f_MSTR",Transform=(Rotation=(X=0,Y=0,Z=0,W=1),Translation=(X=0,Y=0,Z=0),Scale3D=(X=1,Y=1,Z=1))))
    Mode=AdditiveGlobal
    Weight=1
    WeightMaximum=1
    BoneToModify.0.Transform.Translation.X=&RigVMFunction_MathFloatRemap(SourceMaximum=-50, SourceMinimum=0, TargetMaximum=-5, TargetMinimum=0, bClamp=true, Value=@n2.Result.Z).Result
    BoneToModify.0.Transform.Translation.Z=&RigVMFunction_MathFloatRemap(SourceMaximum=-50, SourceMinimum=-30, TargetMaximum=3, TargetMinimum=0, bClamp=true, Value=@n2.Result.Z).Result
    BoneToModify.0.Transform.Rotation=&RigVMFunction_MathQuaternionFromEuler(
     Euler=(X=0,Y=0,Z=0)
     Euler.Y=&RigVMFunction_MathFloatRemap(SourceMaximum=-50, SourceMinimum=-19, TargetMaximum=0, TargetMinimum=0, bClamp=False, Value=@n2.Result.Z).Result
    ).Result
   )
  ]}
]
hierarchy [
 root {loc=<0 0 0>, rot=<0 0 0 1>}
 attach {parent=root, loc=<0 0 0>, rot=<0 0 0 1>}
 pelvis {parent=root, loc=<0 -0.5301 98.0571>, rot=<0.0127 0.707 -0.0127 -0.707>}
 spine_01 {parent=pelvis, loc=<6.1208 -0.9328 0>, rot=<0 0 0.034 -0.9994>}
 spine_02 {parent=spine_01, loc=<6.5996 -0.2095 0>, rot=<0 0 0.0238 -0.9997>}
 spine_03 {parent=spine_02, loc=<6.4914 -0.1237 0>, rot=<0 0 0.0149 -0.9999>}
 spine_04 {parent=spine_03, loc=<6.9598 0.0794 0>, rot=<0 0 -0.0132 -0.9999>}
 spine_05 {parent=spine_04, loc=<6.9987 0 0>, rot=<0 0 -0.0015 -1>}
 clavicle_l {parent=spine_05, loc=<5.4831 -0.1053 -3.2946>, rot=<0.0465 -0.7289 -0.009 -0.6829>}
 clavicle_r {parent=spine_05, loc=<5.4831 -0.1053 3.2946>, rot=<0.7289 0.0465 -0.6829 0.009>}
 neck_01 {parent=spine_05, loc=<11.3138 0.1529 0>, rot=<0 0 -0.2003 -0.9797>}
 neck_02 {parent=neck_01, loc=<4.5627 0 0>, rot=<0 0 0.0504 -0.9987>}
 head {parent=neck_02, loc=<5.3914 0 0>, rot=<0 0 0.115 -0.9934>}
 dyn_bangs_c_01 {parent=head, loc=<8.5955 6.5041 -4.9884>, rot=<0.0878 -0.9429 -0.2831 -0.1519>}
 dyn_hair_a_MSTR {parent=head, loc=<1.7547 -3.2987 -6.6603>, rot=<-0.0738 -0.995 -0.0117 -0.0658>}
 dyn_hair_a_01 {parent=dyn_hair_a_MSTR, loc=<0 0 0>, rot=<0 0 0 -1>}
 dyn_hair_a_02 {parent=dyn_hair_a_01, loc=<4 0 0>, rot=<0 0 0 -1>}
 dyn_hair_a_03 {parent=dyn_hair_a_02, loc=<4 0 0>, rot=<0 0 -0.0619 -0.9981>}
 dyn_hair_a_04 {parent=dyn_hair_a_03, loc=<4 0 0>, rot=<0.0034 -0.0161 -0.2056 -0.9785>}
 dyn_hair_a_05 {parent=dyn_hair_a_04, loc=<3.366 0 0>, rot=<0 0 0 -1>}
 dyn_hair_b_01 {parent=dyn_hair_a_02, loc=<4.1405 1.8534 -1.6007>, rot=<0.1561 -0.0574 -0.3335 -0.928>}
 dyn_hair_b_02 {parent=dyn_hair_b_01, loc=<3.366 0 0>, rot=<0 -0.1876 0 -0.9822>}
 dyn_hair_c_MSTR {parent=head, loc=<-0.3902 -4.7659 -5.6499>, rot=<0.0346 -0.96 0.2378 -0.1436>}
 dyn_hair_c_01 {parent=dyn_hair_c_MSTR, loc=<0 0 0>, rot=<0 0 0 -1>}
 dyn_hair_c_02 {parent=dyn_hair_c_01, loc=<4.7395 0 0>, rot=<0 0 0 -1>}
 dyn_hair_c_03 {parent=dyn_hair_c_02, loc=<4.7395 0 0>, rot=<0 0 0 -1>}
 dyn_hair_d_MSTR {parent=head, loc=<-0.2933 -6.0824 -3.6439>, rot=<0.1014 -0.8214 0.535 -0.1697>}
 dyn_hair_d_01 {parent=dyn_hair_d_MSTR, loc=<0 0 0>, rot=<0 0 0 1>}
 dyn_hair_d_02 {parent=dyn_hair_d_01, loc=<4.2339 0 -0.0136>, rot=<0 0.0063 0 -1>}
 dyn_hair_d_03 {parent=dyn_hair_d_02, loc=<4.2339 0 -0.0136>, rot=<0 -0.0234 0 -0.9997>}
 dyn_hair_d_04 {parent=dyn_hair_d_03, loc=<4.2339 0 -0.0136>, rot=<0 -0.1381 0 -0.9904>}
 dyn_hair_d_05 {parent=dyn_hair_d_04, loc=<4.2343 0 0>, rot=<0 0 0 -1>}
 dyn_hair_e_MSTR {parent=head, loc=<-0.2877 -6.6529 0.0598>, rot=<0.1513 -0.6895 0.6943 -0.1398>}
 dyn_hair_e_01 {parent=dyn_hair_e_MSTR, loc=<0 0 0>, rot=<0 0 0 -1>}
 dyn_hair_e_02 {parent=dyn_hair_e_01, loc=<4.2339 0 -0.0136>, rot=<0 0.0063 0 -1>}
 dyn_hair_e_03 {parent=dyn_hair_e_02, loc=<4.2339 0 -0.0136>, rot=<0 -0.0234 0 -0.9997>}
 dyn_hair_e_04 {parent=dyn_hair_e_03, loc=<4.2339 0 -0.0136>, rot=<0 -0.1381 0 -0.9904>}
 dyn_hair_e_05 {parent=dyn_hair_e_04, loc=<4.2343 0 0>, rot=<0 0 0 1>}
 dyn_hair_f_MSTR {parent=head, loc=<-0.2137 -5.7643 4.2854>, rot=<0.1803 -0.6008 0.7691 -0.1223>}
 dyn_hair_f_01 {parent=dyn_hair_f_MSTR, loc=<0 0 0>, rot=<0 0 0 1>}
 dyn_hair_f_02 {parent=dyn_hair_f_01, loc=<4.2339 0 -0.0136>, rot=<0 -0.0289 0 -0.9996>}
 dyn_hair_f_03 {parent=dyn_hair_f_02, loc=<4.2339 0 -0.0136>, rot=<0 -0.0289 0 -0.9996>}
 dyn_hair_f_04 {parent=dyn_hair_f_03, loc=<4.2339 0 -0.0136>, rot=<0 -0.0289 0 -0.9996>}
 dyn_hair_f_05 {parent=dyn_hair_f_04, loc=<4.2343 0 0>, rot=<0 0 0 1>}
 dyn_hair_g_MSTR {parent=head, loc=<1.5094 -3.1392 6.9169>, rot=<0.0934 -0.2067 0.9736 -0.0256>}
 dyn_hair_g_01 {parent=dyn_hair_g_MSTR, loc=<0 0 0>, rot=<0 0 0 1>}
 dyn_hair_g_02 {parent=dyn_hair_g_01, loc=<4.7395 0 0>, rot=<0 0 0 1>}
 dyn_hair_g_03 {parent=dyn_hair_g_02, loc=<4.7395 0 0>, rot=<0 0 0 1>}
 dyn_hair_h_MSTR {parent=head, loc=<3.9354 0.084 7.6367>, rot=<0.0355 0.0316 0.9988 -0.0083>}
 dyn_hair_h_01 {parent=dyn_hair_h_MSTR, loc=<0 0 0>, rot=<0 0 0 1>}
 dyn_hair_h_02 {parent=dyn_hair_h_01, loc=<4.7546 -0.008 -0.0537>, rot=<0 0 0 1>}
 dyn_hair_h_03 {parent=dyn_hair_h_02, loc=<4.7546 -0.008 -0.0536>, rot=<0 0 0 1>}
 dyn_hair_h_04 {parent=dyn_hair_h_03, loc=<4.7546 -0.008 -0.0537>, rot=<0 0 0.3848 -0.923>}
 dyn_hair_i_MSTR {parent=head, loc=<5.0399 4.8346 8.0229>, rot=<-0.1464 0.2425 0.959 0.0043>}
 dyn_hair_i_01 {parent=dyn_hair_i_MSTR, loc=<0 0 0>, rot=<0 0 0 1>}
 dyn_hair_i_02 {parent=dyn_hair_i_01, loc=<3.366 0 0>, rot=<0 0.1468 0 -0.9892>}
 dyn_bangs_a_MSTR {parent=head, loc=<9.8559 7.9526 2.5838>, rot=<0.2604 0.5267 0.8077 -0.049>}
 dyn_bangs_a_01 {parent=dyn_bangs_a_MSTR, loc=<0 0 0>, rot=<0 0 0 1>}
 dyn_bangs_a_02 {parent=dyn_bangs_a_01, loc=<3.366 0 0>, rot=<0 -0.1876 0 -0.9822>}
 dyn_bangs_b_MSTR {parent=head, loc=<10.097 8.2362 -2.1592>, rot=<0.2666 0.6686 0.6847 -0.1143>}
 dyn_bangs_b_01 {parent=dyn_bangs_b_MSTR, loc=<0 0 0>, rot=<0 0 0 -1>}
 dyn_bangs_b_02 {parent=dyn_bangs_b_01, loc=<2.4109 0 0.3788>, rot=<0 -0.1876 0 -0.9822>}
 dyn_hat_MSTR {parent=head, loc=<14.8215 -4.2416 -0.7229>, rot=<0.4104 -0.5655 0.508 -0.5038>}
 dyn_hat_01 {parent=dyn_hat_MSTR, loc=<0 0 0>, rot=<0 0.1106 0 -0.9939>}
 dyn_hat_02 {parent=dyn_hat_01, loc=<3.366 0 0>, rot=<0 -0.1876 0 -0.9822>}
]
