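# Physics asset for the hair of the female medium Vampire Casual character, covering bangs, lettered hair chains and a hat master bone.
# Capsules on upper spine, clavicles, neck and head form the body shapes, linked by free-rotation constraints.
# Small MSTR anchors join the head with free rotation and root chains with per-strand swing and twist limits.
# Hair b branches from chain a, bangs c attaches directly to the head, and chain i uses boxes.
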PhysicsAsset /BRCosmetics/Characters/Player/Female/Medium/Bodies/F_MED_Vampire_Casual/Meshes/Parts/F_MED_Vampire_Casual_Hair_Physics
bodies {
 clavicle_l [{shapeType=Capsule, length=13, radius=7, center=<3.9437 -0.6222 0.5205>, rotation=<rot 90 0 -8.4719>}]
 clavicle_r [{shapeType=Capsule, length=13, radius=7, center=<-3.4235 1.223 -0.3945>, rotation=<rot 0.0563 -86.3476 91.7762>}]
 dyn_bangs_a_01 [{shapeType=Capsule, length=2.0412, radius=1.5144, shapeName=dyn_bangs_b_01_capsule, center=<1.4921 0 0.0004>, rotation=<rot 0 0 -90>}]
 dyn_bangs_a_02 [{shapeType=Capsule, length=2.7349, radius=0.8207, shapeName=dyn_bangs_b_01_capsule, center=<1.0701 0 0.0004>, rotation=<rot 0 0 -79.5694>}]
 dyn_bangs_a_MSTR [{shapeType=Capsule, length=0.505, radius=0.505}]
 dyn_bangs_b_01 [{shapeType=Capsule, length=2.0412, radius=1.5144, shapeName=dyn_bangs_b_01_capsule, center=<1.0701 0 0.0004>, rotation=<rot 0 0 -90>}]
 dyn_bangs_b_02 [{shapeType=Capsule, length=2.5728, radius=0.9828, shapeName=dyn_bangs_b_01_capsule, center=<1.0701 0 0.0004>, rotation=<rot 0 0 -90>}]
 dyn_bangs_b_MSTR [{shapeType=Capsule, length=0.505, radius=0.505}]
 dyn_bangs_c_01 [{shapeType=Capsule, length=2.4819, radius=1.0737, shapeName=dyn_bangs_b_01_capsule, center=<2.7369 -0.0001 0>, rotation=<rot 0 0 78.6975>}]
 dyn_hair_a_01 [{shapeType=Capsule, length=3.8965, radius=1.1019, shapeName=dyn_hair_a_01_capsule, center=<2.0826 0 -0.0001>, rotation=<rot 0 0 90>}]
 dyn_hair_a_02 [{shapeType=Capsule, length=4.3921, radius=1.0903, shapeName=dyn_hair_e_01_capsule, center=<2.0504 0 0>, rotation=<rot 0 0 90>}]
 dyn_hair_a_03 [{shapeType=Capsule, length=2.7316, radius=1.2684, shapeName=dyn_hair_a_03_capsule, center=<2.0741 0 0>, rotation=<rot 90 0 0>}]
 dyn_hair_a_04 [{shapeType=Capsule, length=2.7316, radius=1.2684, shapeName=dyn_hair_a_04_capsule, center=<2.0741 0 0>, rotation=<rot 90 0 0>}]
 dyn_hair_a_05 [{shapeType=Capsule, length=1.1077, radius=1.2684, shapeName=dyn_hair_a_05_capsule, center=<1.6354 0 -0.0001>, rotation=<rot 90 0 0>}]
 dyn_hair_a_MSTR [{shapeType=Capsule, length=0.505, radius=0.505}]
 dyn_hair_b_01 [{shapeType=Capsule, length=1.7807, radius=0.9822, shapeName=dyn_hair_b_01_capsule, center=<1.523 0 0>, rotation=<rot 90 0 0>}]
 dyn_hair_c_01 [{shapeType=Capsule, length=4.3921, radius=1.0903, shapeName=dyn_hair_e_01_capsule, center=<2.0504 0 0>, rotation=<rot 0 0 90>}]
 dyn_hair_c_02 [{shapeType=Capsule, length=2.4508, radius=2.1312, shapeName=dyn_hair_c_02_capsule, center=<2.2291 0 0>, rotation=<rot 0 0 90>}]
 dyn_hair_c_03 [{shapeType=Capsule, length=6.077, radius=1.8987, center=<1.2815 -1.3053 0.0004>, rotation=<rot 0.0002 38.1818 89.9995>}]
 dyn_hair_c_MSTR [{shapeType=Capsule, length=0.505, radius=0.505}]
 dyn_hair_d_01 [{shapeType=Capsule, length=9, radius=2, shapeName=dyn_hair_e_01_capsule, center=<2.0504 0 0>, rotation=<rot 0 0 90>}]
 dyn_hair_d_02 [{shapeType=Capsule, length=9, radius=2, shapeName=dyn_hair_d_02_capsule, center=<2.2571 0 0>, rotation=<rot 0 0 90>}]
 dyn_hair_d_03 [{shapeType=Capsule, length=9, radius=2, shapeName=dyn_hair_d_03_capsule, center=<2.2548 0 0.0002>, rotation=<rot 0 0 90>}]
 dyn_hair_d_04 [{shapeType=Capsule, length=9, radius=2, shapeName=dyn_hair_d_04_capsule, center=<2.2548 0 0.0002>, rotation=<rot 0 0 90>}]
 dyn_hair_d_05 [{shapeType=Capsule, length=9, radius=2, shapeName=dyn_hair_d_05_capsule, center=<2.2548 0 0.0002>, rotation=<rot 0 0 90>}]
 dyn_hair_d_MSTR [{shapeType=Capsule, length=0.505, radius=0.505}]
 dyn_hair_e_01 [{shapeType=Capsule, length=9.8269, radius=2.5, shapeName=dyn_hair_e_01_capsule, center=<2 0 1.5>, rotation=<rot 0 0 90>}]
 dyn_hair_e_02 [{shapeType=Capsule, length=9.1881, radius=2.5, shapeName=dyn_hair_d_02_capsule, center=<2 0 1.5>, rotation=<rot 0 0 90>}]
 dyn_hair_e_03 [{shapeType=Capsule, length=10.1277, radius=2.5, shapeName=dyn_hair_e_03_capsule, center=<2 0 1.5>, rotation=<rot 0 0 90>}]
 dyn_hair_e_04 [{shapeType=Capsule, length=9.7058, radius=2.5, shapeName=dyn_hair_e_04_capsule, center=<2 0 1.5>, rotation=<rot 0 0 90>}]
 dyn_hair_e_05 [{shapeType=Capsule, length=9.1444, radius=2.5, shapeName=dyn_hair_e_05_capsule, center=<2 0 1.5>, rotation=<rot 0 0 90>}]
 dyn_hair_e_MSTR [{shapeType=Capsule, length=0.505, radius=0.505}]
 dyn_hair_f_01 [{shapeType=Capsule, length=12, radius=2, shapeName=dyn_hair_f_01_capsule, center=<2 0 1.5>, rotation=<rot 0 0 90>}]
 dyn_hair_f_02 [{shapeType=Capsule, length=12, radius=2, shapeName=dyn_hair_f_02_capsule, center=<2 0 1.5>, rotation=<rot 0 0 90>}]
 dyn_hair_f_03 [{shapeType=Capsule, length=12, radius=2, shapeName=dyn_hair_f_03_capsule, center=<2 0 1.5>, rotation=<rot 0 0 90>}]
 dyn_hair_f_04 [{shapeType=Capsule, length=12, radius=2, shapeName=dyn_hair_f_04_capsule, center=<2 0 1.5>, rotation=<rot 0 0 90>}]
 dyn_hair_f_05 [{shapeType=Capsule, length=12, radius=2, shapeName=dyn_hair_f_05_capsule, center=<2 0 1.5>, rotation=<rot 0 0 90>}]
 dyn_hair_f_MSTR [{shapeType=Capsule, length=0.505, radius=0.505}]
 dyn_hair_g_01 [{shapeType=Capsule, length=5.299, radius=0.7708, shapeName=dyn_hair_g_01_capsule, center=<2.1579 0 -0.0001>, rotation=<rot 0 0 90>}]
 dyn_hair_g_02 [{shapeType=Capsule, length=3.6661, radius=0.9362, shapeName=dyn_hair_g_02_capsule, center=<2.196 0 0>, rotation=<rot 0 0 90>}]
 dyn_hair_g_03 [{shapeType=Capsule, length=3.862, radius=0.8793, center=<0.9025 0 0>, rotation=<rot 0 0 90>}]
 dyn_hair_g_MSTR [{shapeType=Capsule, length=0.505, radius=0.505}]
 dyn_hair_h_01 [{shapeType=Capsule, length=5.8697, radius=2, shapeName=dyn_hair_h_01_capsule, center=<2.2048 0 -0.0003>, rotation=<rot 0 0 90>}]
 dyn_hair_h_02 [{shapeType=Capsule, length=2.2007, radius=1.27, shapeName=dyn_hair_h_02_capsule, center=<2.2048 0 -0.0003>, rotation=<rot 0 0 90>}]
 dyn_hair_h_03 [{shapeType=Capsule, length=2.8644, radius=1.27, shapeName=dyn_hair_h_03_capsule, center=<2.2048 0 -0.0003>, rotation=<rot 0 0 90>}]
 dyn_hair_h_04 [{shapeType=Capsule, length=4.4103, radius=1.2579, center=<3.2655 -0.0165 0.3553>, rotation=<rot 76.7288 -11.4548 0>}]
 dyn_hair_h_MSTR [{shapeType=Capsule, length=0.505, radius=0.505}]
 dyn_hair_i_01 [{shapeType=Box, extentX=3.9264, extentY=4.0345, extentZ=1.1776, center=<2.0436 0 0.0001>, rotation=<rot 0.0001 0 6.7815>}]
 dyn_hair_i_02 [{shapeType=Box, extentX=3.9264, extentY=3.2434, extentZ=1.1776, center=<2.0436 0 0.0001>}]
 dyn_hair_i_MSTR [{shapeType=Capsule, length=0.505, radius=0.505}]
 dyn_hat_MSTR [{shapeType=Capsule, length=0.505, radius=0.505}]
 head [
  {shapeType=Capsule, length=5.5862, radius=5.9015, center=<0.8118 3.4354 0>, rotation=<rot 90 0 0>}
  {shapeType=Capsule, length=6.7157, radius=4.772, center=<0.8118 1.1771 0>, rotation=<rot 90 0 0>}
  {shapeType=Capsule, length=9.347, radius=4.522, center=<1.8068 -1.0039 -0.1194>, rotation=<rot 90 0 -10.0004>}
 ]
 neck_01 [{shapeType=Capsule, length=3.7617, radius=5}]
 neck_02 [{shapeType=Capsule, length=2.503, radius=5}]
 spine_04 [{shapeType=Capsule, length=11.2834, radius=8.0962}]
 spine_05 [{shapeType=Capsule, length=24.7679, radius=6.9866}]
}
constraints [
 {bone1Name=dyn_hair_a_01, bone2Name=dyn_hair_a_MSTR, swing1LimitDegrees=15, swing2LimitDegrees=15, twistLimitDegrees=15}
 {bone1Name=dyn_hair_a_02, bone2Name=dyn_hair_a_01, swing1LimitDegrees=15, swing2LimitDegrees=15, twistLimitDegrees=15}
 {bone1Name=dyn_hair_a_03, bone2Name=dyn_hair_a_02, swing1LimitDegrees=15, swing2LimitDegrees=15, twistLimitDegrees=15}
 {bone1Name=dyn_hair_a_04, bone2Name=dyn_hair_a_03, swing1LimitDegrees=15, swing2LimitDegrees=15, twistLimitDegrees=15}
 {bone1Name=dyn_hair_a_05, bone2Name=dyn_hair_a_04, swing1LimitDegrees=15, swing2LimitDegrees=15, twistLimitDegrees=15}
 {bone1Name=dyn_hair_b_01, bone2Name=dyn_hair_a_02, swing1LimitDegrees=15, swing2LimitDegrees=15, twistLimitDegrees=15}
 {bone1Name=dyn_hair_c_01, bone2Name=dyn_hair_c_MSTR, swing1LimitDegrees=25, swing2LimitDegrees=50, twistLimitDegrees=20}
 {bone1Name=dyn_hair_c_02, bone2Name=dyn_hair_c_01, swing1LimitDegrees=25, swing2LimitDegrees=50, twistLimitDegrees=20}
 {bone1Name=dyn_hair_c_03, bone2Name=dyn_hair_c_02, swing1LimitDegrees=25, swing2LimitDegrees=50, twistLimitDegrees=20}
 {bone1Name=dyn_hair_d_01, bone2Name=dyn_hair_d_MSTR, swing1LimitDegrees=25, swing2LimitDegrees=45, twistLimitDegrees=25}
 {bone1Name=dyn_hair_d_02, bone2Name=dyn_hair_d_01, swing1LimitDegrees=25, swing2LimitDegrees=45, twistLimitDegrees=25}
 {bone1Name=dyn_hair_d_03, bone2Name=dyn_hair_d_02, swing1LimitDegrees=25, swing2LimitDegrees=45, twistLimitDegrees=25}
 {bone1Name=dyn_hair_d_04, bone2Name=dyn_hair_d_03, swing1LimitDegrees=25, swing2LimitDegrees=45, twistLimitDegrees=25}
 {bone1Name=dyn_hair_d_05, bone2Name=dyn_hair_d_04, swing1LimitDegrees=25, swing2LimitDegrees=45, twistLimitDegrees=25}
 {bone1Name=dyn_hair_e_01, bone2Name=dyn_hair_e_MSTR, swing1LimitDegrees=5, swing2LimitDegrees=45, twistLimitDegrees=5}
 {bone1Name=dyn_hair_e_02, bone2Name=dyn_hair_e_01, swing1LimitDegrees=5, swing2LimitDegrees=45, twistLimitDegrees=5}
 {bone1Name=dyn_hair_e_03, bone2Name=dyn_hair_e_02, swing1LimitDegrees=5, swing2LimitDegrees=45, twistLimitDegrees=5}
 {bone1Name=dyn_hair_e_04, bone2Name=dyn_hair_e_03, swing1LimitDegrees=5, swing2LimitDegrees=45, twistLimitDegrees=5}
 {bone1Name=dyn_hair_e_05, bone2Name=dyn_hair_e_04, swing1LimitDegrees=5, swing2LimitDegrees=45, twistLimitDegrees=5}
 {bone1Name=dyn_hair_f_01, bone2Name=dyn_hair_f_MSTR, swing1LimitDegrees=30, swing2LimitDegrees=45, twistLimitDegrees=30}
 {bone1Name=dyn_hair_g_01, bone2Name=dyn_hair_g_MSTR, swing1LimitDegrees=25, swing2LimitDegrees=50, twistLimitDegrees=20}
 {bone1Name=dyn_hair_g_02, bone2Name=dyn_hair_g_01, swing1LimitDegrees=25, swing2LimitDegrees=50, twistLimitDegrees=20}
 {bone1Name=dyn_hair_g_03, bone2Name=dyn_hair_g_02, swing1LimitDegrees=25, swing2LimitDegrees=50, twistLimitDegrees=20}
 {bone1Name=dyn_hair_h_01, bone2Name=dyn_hair_h_MSTR, swing1LimitDegrees=30, swing2LimitDegrees=20, twistLimitDegrees=10}
 {bone1Name=dyn_hair_h_02, bone2Name=dyn_hair_h_01, swing1LimitDegrees=30, swing2LimitDegrees=20, twistLimitDegrees=10}
 {bone1Name=dyn_hair_h_03, bone2Name=dyn_hair_h_02, swing1LimitDegrees=30, swing2LimitDegrees=20, twistLimitDegrees=10}
 {bone1Name=dyn_hair_h_04, bone2Name=dyn_hair_h_03, swing1LimitDegrees=30, swing2LimitDegrees=20, twistLimitDegrees=10}
 {bone1Name=dyn_hair_i_01, bone2Name=dyn_hair_i_MSTR, swing1LimitDegrees=25, swing2LimitDegrees=50, twistLimitDegrees=20}
 {bone1Name=dyn_hair_i_02, bone2Name=dyn_hair_i_01, swing1LimitDegrees=25, swing2LimitDegrees=50, twistLimitDegrees=20}
 {
  bone1Name=clavicle_l
  bone2Name=spine_05
  swing1LimitDegrees=45
  swing1Motion=Free
  swing2LimitDegrees=45
  swing2Motion=Free
  twistLimitDegrees=45
  twistMotion=Free
 }
 {
  bone1Name=clavicle_r
  bone2Name=spine_05
  swing1LimitDegrees=45
  swing1Motion=Free
  swing2LimitDegrees=45
  swing2Motion=Free
  twistLimitDegrees=45
  twistMotion=Free
 }
 {
  bone1Name=head
  bone2Name=neck_02
  swing1LimitDegrees=45
  swing1Motion=Free
  swing2LimitDegrees=45
  swing2Motion=Free
  twistLimitDegrees=45
  twistMotion=Free
 }
 {
  bone1Name=spine_05
  bone2Name=spine_04
  swing1LimitDegrees=45
  swing1Motion=Free
  swing2LimitDegrees=45
  swing2Motion=Free
  twistLimitDegrees=45
  twistMotion=Free
 }
 {
  bone1Name=dyn_hair_a_MSTR
  bone2Name=head
  swing1LimitDegrees=45
  swing1Motion=Free
  swing2LimitDegrees=45
  swing2Motion=Free
  twistLimitDegrees=45
  twistMotion=Free
 }
 {
  bone1Name=dyn_hair_c_MSTR
  bone2Name=head
  swing1LimitDegrees=45
  swing1Motion=Free
  swing2LimitDegrees=45
  swing2Motion=Free
  twistLimitDegrees=45
  twistMotion=Free
 }
 {
  bone1Name=dyn_hair_d_MSTR
  bone2Name=head
  swing1LimitDegrees=45
  swing1Motion=Free
  swing2LimitDegrees=45
  swing2Motion=Free
  twistLimitDegrees=45
  twistMotion=Free
 }
 {
  bone1Name=dyn_hair_e_MSTR
  bone2Name=head
  swing1LimitDegrees=45
  swing1Motion=Free
  swing2LimitDegrees=45
  swing2Motion=Free
  twistLimitDegrees=45
  twistMotion=Free
 }
 {
  bone1Name=dyn_hair_f_MSTR
  bone2Name=head
  swing1LimitDegrees=45
  swing1Motion=Free
  swing2LimitDegrees=45
  swing2Motion=Free
  twistLimitDegrees=45
  twistMotion=Free
 }
 {
  bone1Name=dyn_hair_g_MSTR
  bone2Name=head
  swing1LimitDegrees=45
  swing1Motion=Free
  swing2LimitDegrees=45
  swing2Motion=Free
  twistLimitDegrees=45
  twistMotion=Free
 }
 {
  bone1Name=dyn_hair_h_MSTR
  bone2Name=head
  swing1LimitDegrees=45
  swing1Motion=Free
  swing2LimitDegrees=45
  swing2Motion=Free
  twistLimitDegrees=45
  twistMotion=Free
 }
 {
  bone1Name=dyn_hair_i_MSTR
  bone2Name=head
  swing1LimitDegrees=45
  swing1Motion=Free
  swing2LimitDegrees=45
  swing2Motion=Free
  twistLimitDegrees=45
  twistMotion=Free
 }
 {
  bone1Name=dyn_bangs_a_MSTR
  bone2Name=head
  swing1LimitDegrees=45
  swing1Motion=Free
  swing2LimitDegrees=45
  swing2Motion=Free
  twistLimitDegrees=45
  twistMotion=Free
 }
 {
  bone1Name=dyn_bangs_b_MSTR
  bone2Name=head
  swing1LimitDegrees=45
  swing1Motion=Free
  swing2LimitDegrees=45
  swing2Motion=Free
  twistLimitDegrees=45
  twistMotion=Free
 }
 {
  bone1Name=dyn_hat_MSTR
  bone2Name=head
  swing1LimitDegrees=45
  swing1Motion=Free
  swing2LimitDegrees=45
  swing2Motion=Free
  twistLimitDegrees=45
  twistMotion=Free
 }
 {
  bone1Name=neck_01
  bone2Name=spine_05
  swing1LimitDegrees=45
  swing1Motion=Free
  swing2LimitDegrees=45
  swing2Motion=Free
  twistLimitDegrees=45
  twistMotion=Free
 }
 {
  bone1Name=neck_02
  bone2Name=neck_01
  swing1LimitDegrees=45
  swing1Motion=Free
  swing2LimitDegrees=45
  swing2Motion=Free
  twistLimitDegrees=45
  twistMotion=Free
 }
 {bone1Name=dyn_bangs_c_01, bone2Name=head, swing1LimitDegrees=20, swing2LimitDegrees=20, twistLimitDegrees=20}
 {bone1Name=dyn_bangs_a_01, bone2Name=dyn_bangs_a_MSTR, swing1LimitDegrees=20, swing2LimitDegrees=20, twistLimitDegrees=20}
 {bone1Name=dyn_bangs_b_01, bone2Name=dyn_bangs_b_MSTR, swing1LimitDegrees=20, swing2LimitDegrees=20, twistLimitDegrees=20}
 {bone1Name=dyn_bangs_b_02, bone2Name=dyn_bangs_b_01, swing1LimitDegrees=2, swing2LimitDegrees=15, twistLimitDegrees=15}
 {bone1Name=dyn_bangs_a_02, bone2Name=dyn_bangs_a_01, swing1LimitDegrees=20, swing2LimitDegrees=20, twistLimitDegrees=20}
 {bone1Name=dyn_hair_f_02, bone2Name=dyn_hair_f_01, swing1LimitDegrees=30, swing2LimitDegrees=45, twistLimitDegrees=30}
 {bone1Name=dyn_hair_f_03, bone2Name=dyn_hair_f_02, swing1LimitDegrees=30, swing2LimitDegrees=45, twistLimitDegrees=30}
 {bone1Name=dyn_hair_f_04, bone2Name=dyn_hair_f_03, swing1LimitDegrees=30, swing2LimitDegrees=45, twistLimitDegrees=30}
 {bone1Name=dyn_hair_f_05, bone2Name=dyn_hair_f_04, swing1LimitDegrees=30, swing2LimitDegrees=45, twistLimitDegrees=30}
]
