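# Part AnimBlueprint for the female Vampire Casual cosmetic, animating hair, bangs, hat, headband deco and glasses.
# The parent's pose passes through a face-accessory Control Rig, hair RigidBody with gravity override, then per-bone AnimDynamics, constraints and corrective offsets, with glasses aimed by Look At.
# Some hair offsets are toggled by skydiving and rigid-body state.
# The update event sets gravity, skydive, crouch and rifle look-up values; the latter two are unused in this graph.
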AnimBlueprint /BRCosmetics/Characters/Player/Female/Medium/Bodies/F_MED_Vampire_Casual/Meshes/Parts/F_MED_Vampire_Casual_AnimBP
target_skeleton=&/F_MED_Vampire_Casual_Skeleton
graphs [
 AnimGraph {nodes [
   n1 = Control Rig(Source=Copy Pose From Mesh() {bUseAttachedParent=true}) {
    alphaBoolBlend {blendOption=Linear}
    controlRigAssetReference {blueprintRigClass=&/F_MED_Vampire_Casual_FaceAcc_CtrlRig.F_MED_Vampire_Casual_FaceAcc_CtrlRig_C}
   }
   n2 = RigidBody!(ComponentPose=Local To Component(LocalPose=@n1), Alpha, OverrideWorldGravity:Vector=Get GravityOverride(), ExternalForce) -> (Pose) {
    bClampLinearTranslationLimitToRefPose=true
    bOverrideWorldGravity=true
    overridePhysicsAsset=&/F_MED_Vampire_Casual_Hair_Physics
    simulationSpace=BaseBoneSpace
    alphaBoolBlend {blendOption=Linear}
    baseBoneRef {boneName=neck_01}
    componentAppliedLinearAccClamp=<2000 2000 2000>
    componentLinearAccScale=<0.3 0.4 0.3>
    componentLinearVelScale=<0.3 0.4 0.3>
    overrideWorldGravity {z=-980}
   }
   n3 = Anim Dynamics - Bone: dyn_hair_h_01(ComponentPose=@n2, Alpha=0.5, ExternalForce="0,0,0") {
    angularDampingOverride=0.9999
    angularSpringConstant=75
    bAngularSpring=true
    bGravityOverrideInSimSpace=true
    bOverrideAngularDamping=true
    simulationSpace=RootRelative
    alphaBoolBlend {blendOption=Linear}
    boundBone {boneName=dyn_hair_h_01}
    componentLinearAccScale=<1 1 1>
    componentLinearVelScale=<1 1 1>
    physicsBodyDefinitions [{
      sphereCollisionRadius=0
      boundBone {boneName=dyn_hair_h_01}
      boxExtents=<10 20 20>
      constraintSetup {angularLimitsMax=<5 0 5>, angularLimitsMin=<-5 -45 -45>, angularTarget=<0 0 0>}
      localJointOffset=<1 0 -0.2>
     }]
   }
   n4 = Anim Dynamics - Bone: dyn_hair_h_02(ComponentPose=@n3, Alpha=0.5, ExternalForce="0,0,0") {
    angularDampingOverride=0.9999
    angularSpringConstant=75
    bAngularSpring=true
    bGravityOverrideInSimSpace=true
    bOverrideAngularDamping=true
    simulationSpace=RootRelative
    alphaBoolBlend {blendOption=Linear}
    boundBone {boneName=dyn_hair_h_02}
    componentLinearAccScale=<1 1 1>
    componentLinearVelScale=<1 1 1>
    physicsBodyDefinitions [{
      sphereCollisionRadius=0
      boundBone {boneName=dyn_hair_h_02}
      boxExtents=<10 20 20>
      constraintSetup {angularLimitsMax=<5 0 5>, angularLimitsMin=<-5 -45 -45>, angularTarget=<0 0 0>}
      localJointOffset=<1 0 -0.2>
     }]
   }
   n5 = Anim Dynamics - Bone: dyn_hair_h_03(ComponentPose=@n4, Alpha=0.5, ExternalForce="0,0,0") {
    angularDampingOverride=0.9999
    angularSpringConstant=75
    bAngularSpring=true
    bGravityOverrideInSimSpace=true
    bOverrideAngularDamping=true
    simulationSpace=RootRelative
    alphaBoolBlend {blendOption=Linear}
    boundBone {boneName=dyn_hair_h_03}
    componentLinearAccScale=<1 1 1>
    componentLinearVelScale=<1 1 1>
    physicsBodyDefinitions [{
      sphereCollisionRadius=0
      boundBone {boneName=dyn_hair_h_03}
      boxExtents=<10 20 20>
      constraintSetup {angularLimitsMax=<5 0 5>, angularLimitsMin=<-5 -45 -45>, angularTarget=<0 0 0>}
      localJointOffset=<1 0 -0.2>
     }]
   }
   n6 = Anim Dynamics - Bone: dyn_hair_h_04(ComponentPose=@n5, Alpha=0.5, ExternalForce="0,0,0") {
    angularDampingOverride=0.9999
    angularSpringConstant=75
    bAngularSpring=true
    bGravityOverrideInSimSpace=true
    bOverrideAngularDamping=true
    simulationSpace=RootRelative
    alphaBoolBlend {blendOption=Linear}
    boundBone {boneName=dyn_hair_h_04}
    componentLinearAccScale=<1 1 1>
    componentLinearVelScale=<1 1 1>
    physicsBodyDefinitions [{
      sphereCollisionRadius=0
      boundBone {boneName=dyn_hair_h_04}
      boxExtents=<10 20 20>
      constraintSetup {angularLimitsMax=<5 0 5>, angularLimitsMin=<-5 -45 -45>, angularTarget=<0 0 0>}
      localJointOffset=<1 0 -0.2>
     }]
   }
   n7 = Anim Dynamics - Bone: dyn_hair_a_01(ComponentPose=@n6, Alpha=0.2, ExternalForce="0,0,0") {
    angularDampingOverride=0.9999
    angularSpringConstant=75
    bAngularSpring=true
    bGravityOverrideInSimSpace=true
    bOverrideAngularDamping=true
    simulationSpace=RootRelative
    alphaBoolBlend {blendOption=Linear}
    boundBone {boneName=dyn_hair_a_01}
    componentLinearAccScale=<1 1 1>
    componentLinearVelScale=<1 1 1>
    physicsBodyDefinitions [{
      sphereCollisionRadius=0
      boundBone {boneName=dyn_hair_a_01}
      boxExtents=<10 20 20>
      constraintSetup {angularLimitsMax=<45 45 45>, angularLimitsMin=<-45 -45 -45>, angularTarget=<0 0 0>}
      localJointOffset=<1 0 0>
     }]
    relativeSpaceBone {boneName=neck_01}
   }
   n8 = Anim Dynamics - Bone: dyn_hair_a_02(ComponentPose=@n7, Alpha=0.2, ExternalForce="0,0,0") {
    angularDampingOverride=0.9999
    angularSpringConstant=75
    bAngularSpring=true
    bGravityOverrideInSimSpace=true
    bOverrideAngularDamping=true
    simulationSpace=RootRelative
    alphaBoolBlend {blendOption=Linear}
    boundBone {boneName=dyn_hair_a_02}
    componentLinearAccScale=<1 1 1>
    componentLinearVelScale=<1 1 1>
    physicsBodyDefinitions [{
      sphereCollisionRadius=0
      boundBone {boneName=dyn_hair_a_02}
      boxExtents=<10 20 20>
      constraintSetup {angularLimitsMax=<45 45 45>, angularLimitsMin=<-45 -45 -45>, angularTarget=<0 0 0>}
      localJointOffset=<1 0 0>
     }]
    relativeSpaceBone {boneName=neck_01}
   }
   n9 = Anim Dynamics - Bone: dyn_hair_a_03(ComponentPose=@n8, Alpha=0.2, ExternalForce="0,0,0") {
    angularDampingOverride=0.9999
    angularSpringConstant=75
    bAngularSpring=true
    bGravityOverrideInSimSpace=true
    bOverrideAngularDamping=true
    simulationSpace=RootRelative
    alphaBoolBlend {blendOption=Linear}
    boundBone {boneName=dyn_hair_a_03}
    componentLinearAccScale=<1 1 1>
    componentLinearVelScale=<1 1 1>
    physicsBodyDefinitions [{
      sphereCollisionRadius=0
      boundBone {boneName=dyn_hair_a_03}
      boxExtents=<10 20 20>
      constraintSetup {angularLimitsMax=<45 45 45>, angularLimitsMin=<-45 -45 -45>, angularTarget=<0 0 0>}
      localJointOffset=<1 0 0>
     }]
    relativeSpaceBone {boneName=neck_01}
   }
   n10 = Anim Dynamics - Bone: dyn_hair_a_04(ComponentPose=@n9, Alpha=0.2, ExternalForce="0,0,0") {
    angularDampingOverride=0.9999
    angularSpringConstant=75
    bAngularSpring=true
    bGravityOverrideInSimSpace=true
    bOverrideAngularDamping=true
    simulationSpace=RootRelative
    alphaBoolBlend {blendOption=Linear}
    boundBone {boneName=dyn_hair_a_04}
    componentLinearAccScale=<1 1 1>
    componentLinearVelScale=<1 1 1>
    physicsBodyDefinitions [{
      sphereCollisionRadius=0
      boundBone {boneName=dyn_hair_a_04}
      boxExtents=<10 20 20>
      constraintSetup {angularLimitsMax=<45 45 45>, angularLimitsMin=<-45 -45 -45>, angularTarget=<0 0 0>}
      localJointOffset=<1 0 0>
     }]
    relativeSpaceBone {boneName=neck_01}
   }
   n11 = Anim Dynamics - Bone: dyn_hair_a_05(ComponentPose=@n10, Alpha=0.2, ExternalForce="0,0,0") {
    angularDampingOverride=0.9999
    angularSpringConstant=75
    bAngularSpring=true
    bGravityOverrideInSimSpace=true
    bOverrideAngularDamping=true
    simulationSpace=RootRelative
    alphaBoolBlend {blendOption=Linear}
    boundBone {boneName=dyn_hair_a_05}
    componentLinearAccScale=<1 1 1>
    componentLinearVelScale=<1 1 1>
    physicsBodyDefinitions [{
      sphereCollisionRadius=0
      boundBone {boneName=dyn_hair_a_05}
      boxExtents=<10 20 20>
      constraintSetup {angularLimitsMax=<45 45 45>, angularLimitsMin=<-45 -45 -45>, angularTarget=<0 0 0>}
      localJointOffset=<1 0 0>
     }]
    relativeSpaceBone {boneName=neck_01}
   }
   n12 = "Constraint - dyn_hair_a_04 (P)"(ComponentPose=@n11, Alpha=0.8, ConstraintWeights_0=1) {
    alphaBoolBlend {blendOption=Linear}
    boneToModify {boneName=dyn_hair_a_04}
    constraintSetup [{targetBone {boneName=clavicle_l}}]
    constraintWeights [1]
   }
   n13 = "Constraint - dyn_hair_h_03 (P)"(ComponentPose=@n12, Alpha=0.5, ConstraintWeights_0=1) {
    alphaBoolBlend {blendOption=Linear}
    boneToModify {boneName=dyn_hair_h_03}
    constraintSetup [{targetBone {boneName=clavicle_r}}]
    constraintWeights [1]
   }
   n14 = "Constraint - dyn_hair_b_01 (P)"(ComponentPose=@n13, Alpha=0.8, ConstraintWeights_0=1) {
    alphaBoolBlend {blendOption=Linear}
    boneToModify {boneName=dyn_hair_b_01}
    constraintSetup [{targetBone {boneName=neck_01}}]
    constraintWeights [1]
   }
   n15 = "Constraint - dyn_hair_g_01 (PP)"!(ComponentPose=@n14, Alpha=1, ConstraintWeights_0=0.75, ConstraintWeights_1:float=0.25) -> (Pose) {
    alphaBoolBlend {blendOption=Linear}
    boneToModify {boneName=dyn_hair_g_01}
    constraintSetup [{targetBone {boneName=dyn_hair_f_01}}, {targetBone {boneName=dyn_hair_h_01}}]
    constraintWeights [1,1]
   }
   n16 = "Transform (Modify) Bone - Bone: dyn_hair_g_MSTR"(ComponentPose=@n15, Alpha=1, Translation="0,0,6.5", Rotation="0,0,0", Scale="1,1,1") {translationMode=BMM_Additive, alphaBoolBlend {blendOption=Linear}, boneToModify {boneName=dyn_hair_g_MSTR}, translation {z=6.5}}
   n17 = "Constraint - dyn_hair_g_MSTR (P)"(ComponentPose=@n16, Alpha=0.5, ConstraintWeights_0=1) {
    alphaBoolBlend {blendOption=Linear}
    boneToModify {boneName=dyn_hair_g_MSTR}
    constraintSetup [{targetBone {boneName=clavicle_r}}]
    constraintWeights [1]
   }
   n18 = "Transform (Modify) Bone - Bone: dyn_hair_g_MSTR"(ComponentPose=@n17, Alpha=1, Translation="-2,0,2", Rotation="20,0,0", Scale="1,1,1") {
    rotationMode=BMM_Additive
    scaleMode=BMM_Additive
    translationMode=BMM_Additive
    alphaBoolBlend {blendOption=Linear}
    boneToModify {boneName=dyn_hair_g_MSTR}
    rotation {pitch=20}
    translation {x=-2, z=2}
   }
   n19 = "Transform (Modify) Bone - Bone: dyn_hair_f_MSTR"!(ComponentPose=@n18, Alpha=1, Rotation="-5.000,0.000,-0.000") -> (Pose) {
    rotationMode=BMM_Additive
    scaleMode=BMM_Additive
    translationMode=BMM_Additive
    alphaBoolBlend {blendOption=Linear}
    boneToModify {boneName=dyn_hair_f_MSTR}
    rotation {pitch=-5}
   }
   n20 = "Transform (Modify) Bone - Bone: dyn_hair_c_MSTR"!(ComponentPose=@n19, Alpha=1, Rotation="5.000,0.000,0.000") -> (Pose) {
    rotationMode=BMM_Additive
    scaleMode=BMM_Additive
    translationMode=BMM_Additive
    alphaBoolBlend {blendOption=Linear}
    boneToModify {boneName=dyn_hair_c_MSTR}
    rotation {pitch=5}
   }
   n21 = "Transform (Modify) Bone - Bone: dyn_hair_d_MSTR"!(ComponentPose=@n20, Alpha=1, Rotation="-4.981,-0.437,5.019") -> (Pose) {
    rotationMode=BMM_Additive
    scaleMode=BMM_Additive
    translationMode=BMM_Additive
    alphaBoolBlend {blendOption=Linear}
    boneToModify {boneName=dyn_hair_d_MSTR}
    rotation=<rot -4.981 -0.437 5.019>
   }
   n22 = "Transform (Modify) Bone - Bone: dyn_hair_c_MSTR"!(ComponentPose=@n21, Alpha=1, Translation:Vector="0,0,0", Rotation="-3,0,0", Scale:Vector="1,1,1") -> (Pose) {
    rotationMode=BMM_Additive
    scaleMode=BMM_Additive
    translationMode=BMM_Additive
    alphaBoolBlend {blendOption=Linear}
    boneToModify {boneName=dyn_hair_c_MSTR}
    rotation {pitch=-3}
   }
   n23 = Get Skydiving() -> (Skydiving:bool)
   n24 = "Transform (Modify) Bone - Bone: dyn_hair_c_MSTR"!(ComponentPose=@n22, bAlphaBoolEnabled:bool=@n23, Translation="0,0,3", Rotation="0,0,0", Scale="1,1,1") -> (Pose) {
    alphaInputType=Bool
    rotationMode=BMM_Additive
    scaleMode=BMM_Additive
    translationMode=BMM_Additive
    alphaBoolBlend {blendOption=Linear}
    boneToModify {boneName=dyn_hair_c_MSTR}
    translation {z=3}
   }
   n25 = "Transform (Modify) Bone - Bone: dyn_hair_d_MSTR"!(ComponentPose=@n24, bAlphaBoolEnabled:bool=@n23, Translation:Vector="0,0,3", Rotation="0,0,0", Scale:Vector="1,1,1") -> (Pose) {
    alphaInputType=Bool
    rotationMode=BMM_Additive
    scaleMode=BMM_Additive
    translationMode=BMM_Additive
    alphaBoolBlend {blendOption=Linear}
    boneToModify {boneName=dyn_hair_d_MSTR}
    translation {z=3}
   }
   n26 = Anim Dynamics - Bone: dyn_hat_01(ComponentPose=@n25, Alpha=0.7, ExternalForce="0,0,0") {
    angularDampingOverride=0.9999
    angularSpringConstant=1500
    bAngularSpring=true
    bGravityOverrideInSimSpace=true
    bOverrideAngularDamping=true
    simulationSpace=RootRelative
    alphaBoolBlend {blendOption=Linear}
    boundBone {boneName=dyn_hat_01}
    componentLinearAccScale=<1 1 1>
    componentLinearVelScale=<1 1 1>
    physicsBodyDefinitions [{
      sphereCollisionRadius=0
      boundBone {boneName=dyn_hat_01}
      boxExtents=<20 20 10>
      constraintSetup {angularLimitsMax=<20 2 20>, angularLimitsMin=<-20 -45 -45>, angularTarget=<0 0 0>}
      localJointOffset=<0 0 -2>
     }]
   }
   n27 = Anim Dynamics - Bone: dyn_hat_02(ComponentPose=@n26, Alpha=0.7, ExternalForce="0,0,0") {
    angularDampingOverride=0.9999
    angularSpringConstant=5000
    bAngularSpring=true
    bGravityOverrideInSimSpace=true
    bOverrideAngularDamping=true
    simulationSpace=RootRelative
    alphaBoolBlend {blendOption=Linear}
    boundBone {boneName=dyn_hat_02}
    componentLinearAccScale=<1 1 1>
    componentLinearVelScale=<1 1 1>
    physicsBodyDefinitions [{
      sphereCollisionRadius=0
      boundBone {boneName=dyn_hat_02}
      boxExtents=<20 20 10>
      constraintSetup {angularLimitsMax=<20 2 20>, angularLimitsMin=<-20 -45 -45>, angularTarget=<0 0 0>}
      localJointOffset=<0.5 0 -2>
     }]
   }
   n28 = Anim Dynamics - Bone: dyn_headBandDeco(ComponentPose=@n27, Alpha=1, ExternalForce="0,0,0") {
    angularSpringConstant=250
    bAngularSpring=true
    bGravityOverrideInSimSpace=true
    bOverrideAngularDamping=true
    gravityScale=0.5
    alphaBoolBlend {blendOption=Linear}
    boundBone {boneName=dyn_headBandDeco}
    physicsBodyDefinitions [{
      sphereCollisionRadius=0
      boundBone {boneName=dyn_headBandDeco}
      boxExtents=<10 10 5>
      constraintSetup {angularLimitsMax=<20 1 5>, angularLimitsMin=<-20 -1 -5>, angularTarget=<0 0 0>}
      localJointOffset=<0 0 -3>
     }]
   }
   n29 = Anim Dynamics - Bone: dyn_hair_h_01(ComponentPose=@n28, Alpha=1, ExternalForce="0,0,0") {
    angularDampingOverride=0.9999
    angularSpringConstant=75
    bAngularSpring=true
    bGravityOverrideInSimSpace=true
    bOverrideAngularDamping=true
    simulationSpace=RootRelative
    alphaBoolBlend {blendOption=Linear}
    boundBone {boneName=dyn_hair_h_01}
    componentLinearAccScale=<1 1 1>
    componentLinearVelScale=<1 1 1>
    physicsBodyDefinitions [{
      sphereCollisionRadius=0
      boundBone {boneName=dyn_hair_h_01}
      boxExtents=<20 20 10>
      constraintSetup {angularLimitsMax=<25 20 0>, angularLimitsMin=<-25 -25 -15>, angularTarget=<0 0 0>}
      localJointOffset=<2 0 -2>
     }]
   }
   n30 = Anim Dynamics - Bone: dyn_hair_i_01(ComponentPose=@n29, Alpha=1, ExternalForce="0,0,0") {
    angularDampingOverride=0.9999
    angularSpringConstant=75
    bAngularSpring=true
    bGravityOverrideInSimSpace=true
    bOverrideAngularDamping=true
    simulationSpace=RootRelative
    alphaBoolBlend {blendOption=Linear}
    boundBone {boneName=dyn_hair_i_01}
    componentLinearAccScale=<1 1 1>
    componentLinearVelScale=<1 1 1>
    physicsBodyDefinitions [{
      sphereCollisionRadius=0
      boundBone {boneName=dyn_hair_i_01}
      boxExtents=<20 20 10>
      constraintSetup {angularLimitsMax=<25 20 15>, angularLimitsMin=<-25 -25 -15>, angularTarget=<0 0 0>}
      localJointOffset=<2 0 0>
     }]
   }
   n31 = Anim Dynamics - Bone: dyn_hair_i_02(ComponentPose=@n30, Alpha=1, ExternalForce="0,0,0") {
    angularDampingOverride=0.9999
    angularSpringConstant=75
    bAngularSpring=true
    bGravityOverrideInSimSpace=true
    bOverrideAngularDamping=true
    simulationSpace=RootRelative
    alphaBoolBlend {blendOption=Linear}
    boundBone {boneName=dyn_hair_i_02}
    componentLinearAccScale=<1 1 1>
    componentLinearVelScale=<1 1 1>
    physicsBodyDefinitions [{
      sphereCollisionRadius=0
      boundBone {boneName=dyn_hair_i_02}
      boxExtents=<20 20 10>
      constraintSetup {angularLimitsMax=<25 20 15>, angularLimitsMin=<-25 -25 -15>, angularTarget=<0 0 0>}
      localJointOffset=<2 0 0>
     }]
   }
   n32 = Anim Dynamics - Bone: dyn_hair_b_01(ComponentPose=@n31, Alpha=1, ExternalForce="0,0,0") {
    angularDampingOverride=0.9999
    angularSpringConstant=75
    bAngularSpring=true
    bGravityOverrideInSimSpace=true
    bOverrideAngularDamping=true
    simulationSpace=RootRelative
    alphaBoolBlend {blendOption=Linear}
    boundBone {boneName=dyn_hair_b_01}
    componentLinearAccScale=<1 1 1>
    componentLinearVelScale=<1 1 1>
    physicsBodyDefinitions [{
      sphereCollisionRadius=0
      boundBone {boneName=dyn_hair_b_01}
      boxExtents=<20 20 10>
      constraintSetup {angularLimitsMax=<25 20 15>, angularLimitsMin=<-25 -25 -15>, angularTarget=<0 0 0>}
      localJointOffset=<2 0 0>
     }]
   }
   n33 = Anim Dynamics - Bone: dyn_hair_b_02(ComponentPose=@n32, Alpha=1, ExternalForce="0,0,0") {
    angularDampingOverride=0.9999
    angularSpringConstant=75
    bAngularSpring=true
    bGravityOverrideInSimSpace=true
    bOverrideAngularDamping=true
    simulationSpace=RootRelative
    alphaBoolBlend {blendOption=Linear}
    boundBone {boneName=dyn_hair_b_02}
    componentLinearAccScale=<1 1 1>
    componentLinearVelScale=<1 1 1>
    physicsBodyDefinitions [{
      sphereCollisionRadius=0
      boundBone {boneName=dyn_hair_b_02}
      boxExtents=<20 20 10>
      constraintSetup {angularLimitsMax=<25 20 15>, angularLimitsMin=<-25 -25 -15>, angularTarget=<0 0 0>}
      localJointOffset=<2 0 0>
     }]
   }
   n34 = Anim Dynamics - Bone: dyn_bangs_a_01(ComponentPose=@n33, Alpha=1, ExternalForce="0,0,0") {
    angularDampingOverride=0.9999
    angularSpringConstant=75
    bAngularSpring=true
    bGravityOverrideInSimSpace=true
    bOverrideAngularDamping=true
    simulationSpace=RootRelative
    alphaBoolBlend {blendOption=Linear}
    boundBone {boneName=dyn_bangs_a_01}
    componentLinearAccScale=<1 1 1>
    componentLinearVelScale=<1 1 1>
    physicsBodyDefinitions [{
      sphereCollisionRadius=0
      boundBone {boneName=dyn_bangs_a_01}
      boxExtents=<20 20 10>
      constraintSetup {angularLimitsMax=<25 0 15>, angularLimitsMin=<-25 -25 -15>, angularTarget=<0 0 0>}
      localJointOffset=<2 0 0>
     }]
   }
   n35 = Anim Dynamics - Bone: dyn_bangs_a_02(ComponentPose=@n34, Alpha=1, ExternalForce="0,0,0") {
    angularDampingOverride=0.9999
    angularSpringConstant=75
    bAngularSpring=true
    bGravityOverrideInSimSpace=true
    bOverrideAngularDamping=true
    simulationSpace=RootRelative
    alphaBoolBlend {blendOption=Linear}
    boundBone {boneName=dyn_bangs_a_02}
    componentLinearAccScale=<1 1 1>
    componentLinearVelScale=<1 1 1>
    physicsBodyDefinitions [{
      sphereCollisionRadius=0
      boundBone {boneName=dyn_bangs_a_02}
      boxExtents=<20 20 10>
      constraintSetup {angularLimitsMax=<25 0 15>, angularLimitsMin=<-25 -25 -15>, angularTarget=<0 0 0>}
      localJointOffset=<2 0 0>
     }]
   }
   n36 = Anim Dynamics - Bone: dyn_bangs_b_01(ComponentPose=@n35, Alpha=1, ExternalForce="0,0,0") {
    angularDampingOverride=0.9999
    angularSpringConstant=75
    bAngularSpring=true
    bGravityOverrideInSimSpace=true
    bOverrideAngularDamping=true
    simulationSpace=RootRelative
    alphaBoolBlend {blendOption=Linear}
    boundBone {boneName=dyn_bangs_b_01}
    componentLinearAccScale=<1 1 1>
    componentLinearVelScale=<1 1 1>
    physicsBodyDefinitions [{
      sphereCollisionRadius=0
      boundBone {boneName=dyn_bangs_b_01}
      boxExtents=<20 20 10>
      constraintSetup {angularLimitsMax=<25 0 15>, angularLimitsMin=<-25 -25 -15>, angularTarget=<0 0 0>}
      localJointOffset=<2 0 0>
     }]
   }
   n37 = Anim Dynamics - Bone: dyn_bangs_b_02(ComponentPose=@n36, Alpha=1, ExternalForce="0,0,0") {
    angularDampingOverride=0.9999
    angularSpringConstant=75
    bAngularSpring=true
    bGravityOverrideInSimSpace=true
    bOverrideAngularDamping=true
    simulationSpace=RootRelative
    alphaBoolBlend {blendOption=Linear}
    boundBone {boneName=dyn_bangs_b_02}
    componentLinearAccScale=<1 1 1>
    componentLinearVelScale=<1 1 1>
    physicsBodyDefinitions [{
      sphereCollisionRadius=0
      boundBone {boneName=dyn_bangs_b_02}
      boxExtents=<20 20 10>
      constraintSetup {angularLimitsMax=<25 0 15>, angularLimitsMin=<-25 -25 -15>, angularTarget=<0 0 0>}
      localJointOffset=<2 0 0>
     }]
   }
   n38 = Anim Dynamics - Bone: dyn_bangs_c_01(ComponentPose=@n37, Alpha=1, ExternalForce="0,0,0") {
    angularDampingOverride=0.9999
    angularSpringConstant=75
    bAngularSpring=true
    bGravityOverrideInSimSpace=true
    bOverrideAngularDamping=true
    simulationSpace=RootRelative
    alphaBoolBlend {blendOption=Linear}
    boundBone {boneName=dyn_bangs_c_01}
    componentLinearAccScale=<1 1 1>
    componentLinearVelScale=<1 1 1>
    physicsBodyDefinitions [{
      sphereCollisionRadius=0
      boundBone {boneName=dyn_bangs_c_01}
      boxExtents=<20 20 10>
      constraintSetup {angularLimitsMax=<25 0 15>, angularLimitsMin=<-25 -25 -15>, angularTarget=<0 0 0>}
      localJointOffset=<2 0 0>
     }]
   }
   n39 = "Transform (Modify) Bone - Bone: dyn_hair_h_MSTR"!(ComponentPose=@n38, bAlphaBoolEnabled:bool=Get bRigidBodyEnabled(), Translation="0,0,0", Rotation="-30,0,0", Scale="1,1,1") -> (Pose) {
    alphaInputType=Bool
    rotationMode=BMM_Additive
    rotationSpace=BCS_BoneSpace
    scaleMode=BMM_Additive
    scaleSpace=BCS_BoneSpace
    translationMode=BMM_Additive
    translationSpace=BCS_BoneSpace
    alphaBoolBlend {blendOption=Linear}
    boneToModify {boneName=dyn_hair_h_MSTR}
    rotation {pitch=-30}
   }
   n40 = Look At - Bone: dyn_glasses(ComponentPose=@n39, Alpha=1) {
    alphaBoolBlend {blendOption=Linear}
    boneToModify {boneName=dyn_glasses}
    lookAtLocation {x=2.2}
    lookAtTarget {boneReference {boneName=head}}
    lookAt_Axis {axis=<-1 0>}
   }
   n41 = "Transform (Modify) Bone - Bone: dyn_glasses"(ComponentPose=@n40, Alpha=1, Translation="0.25,0,0.15", Rotation="0,0,0", Scale="1,1,1") {
    translationMode=BMM_Additive
    translationSpace=BCS_BoneSpace
    alphaBoolBlend {blendOption=Linear}
    boneToModify {boneName=dyn_glasses}
    translation {x=0.25, z=0.15}
   }
   Output Pose(Result=Component To Local(ComponentPose=@n41)) {layerGroup=None}
  ]}
 EventGraph {nodes [
   Event BlueprintUpdateAnimation()
   > n1 = Sequence()
   > Set Crouching(Crouching:bool=Get bIsCrouching()) -> (Output_Get:bool)
   > Set Skydiving(Skydiving:bool=OR Boolean!(A=Get bIsSkyDiving(), B=Get bIsSkydiveFloating(), C:bool=Get bIsSkydiveDiveMode()) -> (ReturnValue)) -> (Output_Get:bool)
   n2 = Gravity Override State Function DEPRECATED(
    execute=@n1.then_1
    CharacterPartAnimInstance=Self-Reference()
    GameTimeInSeconds=GetGameTimeInSeconds()
    GravityOverrideParams=Get Gravity Override Params()
   )
   > Set GravityOverride(GravityOverride=@n2)
   n3 = BreakTransform(InTransform=GetDeltaTransformFromRefPose(self=GetOwnerSkeletalMeshComponent(), BoneName=head, BaseName=neck_02))
   > Set RifleLookUp(RifleLookUp:double=MapRangeClamped(Value=BreakRotator(InRot=@n3.Rotation).Yaw, InRangeA=5, InRangeB=-20)) -> (Output_Get:double)
  ]}
]
variables [
 Crouching {type=boolean}
 Gravity Override Params {type=GravityOverrideParamsStruct}
 GravityOverride {type=Vector}
 RifleLookUp {type=number}
 Skydiving {type=boolean}
]
